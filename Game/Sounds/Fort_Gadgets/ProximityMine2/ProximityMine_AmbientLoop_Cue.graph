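# Sound cue for the proximity mine gadget that plays an activation one-shot together with a looping ambient sound, in the Fort_Explosives sound class.
# A mixer blends two branches, each with its own attenuation asset, with the loop mixed lower than the activation.
# Each branch randomly picks a wave and passes it through a modulator, which varies pitch upward on the loop.
# Cue-level concurrency overrides cap simultaneous instances.
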SoundCue /Game/Sounds/Fort_Gadgets/ProximityMine2/ProximityMine_AmbientLoop_Cue
root=Mixer_0
nodes [
 {id=Attenuation_0, children [Modulator_0], properties {attenuationSettings=/Game/Sounds/Attenuation/Fort_Explosive_Fuse_Ambient_Att}}
 {id=Attenuation_1, children [Modulator_1], properties {attenuationSettings=/Game/Sounds/Attenuation/Fort_Weapon_Proximity_Mine_Att}}
 {id=Mixer_0, children [Attenuation_1, Attenuation_0], properties {inputVolume [1,0.4]}}
 {id=Modulator_0, children [Random_3], properties {pitchMax=1.2, pitchMin=1}}
 {id=Modulator_1, children [Random_2]}
 {id=Random_2, children [WavePlayer_5,WavePlayer_6,WavePlayer_7], properties {weights [1,1,1]}}
 {id=Random_3, children [WavePlayer_10, WavePlayer_11], properties {weights [1,1]}}
 {id=WavePlayer_10, properties {bLooping=true, soundWaveAssetPtr=&/z_ProximityMine_AmbientLoop_02}}
 {id=WavePlayer_11, properties {bLooping=true, soundWaveAssetPtr=&/z_ProximityMine_AmbientLoop_03}}
 {id=WavePlayer_5, properties {soundWaveAssetPtr=&/ProximityMine_Activate_01}}
 {id=WavePlayer_6, properties {soundWaveAssetPtr=&/ProximityMine_Activate_02}}
 {id=WavePlayer_7, properties {soundWaveAssetPtr=&/ProximityMine_Activate_03}}
]
properties {
 bOverrideConcurrency=true
 maxDistance=3500
 soundClassObject=/Game/Sounds/FortSoundClassesAndModes/Fort_Explosives
 volumeMultiplier=0.3
 attenuationOverrides {distanceAlgorithm=Logarithmic, falloffDistance=1024, stereoSpread=1, attenuationShapeExtents {x=150}}
 concurrencyOverrides {maxCount=4}
}
