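# SoundCue that plays the proximity mine gadget's beep in the game.
# A single wave player for ProximityMine_Beep_02 feeds the output directly, at reduced volume with a slight pitch shift.
# It references the shared working-gadget attenuation asset but enables inline attenuation and concurrency overrides, limiting range and simultaneous instances.
# It is routed through the third-person ranged weapon fire sound class.
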SoundCue /Game/Sounds/Fort_Gadgets/ProximityMine2/ProximityMine_Beep_Cue
root=WavePlayer_2
nodes [{id=WavePlayer_2, properties {soundWaveAssetPtr=&/ProximityMine_Beep_02}}]
properties {
 attenuationSettings=/Game/Sounds/Attenuation/Fort_GadgetWorking_Att
 bOverrideAttenuation=true
 bOverrideConcurrency=true
 maxDistance=1650
 pitchMultiplier=1.02
 soundClassObject=&Fort_WeaponRanged_Fire_3P_PSM
 volumeMultiplier=0.25
 attenuationOverrides {falloffDistance=1200, stereoSpread=1, attenuationShapeExtents {x=450}}
 concurrencyOverrides {maxCount=3}
}
