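# Sound cue that plays when a proximity mine gadget is disarmed.
# A single wave player outputs one disarm sample directly, with no randomization or modulation.
# It uses the proximity mine weapon attenuation asset, plays at reduced volume, and routes through the loot-collect sound class.
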SoundCue /Game/Sounds/Fort_Gadgets/ProximityMine2/ProximityMine_Disarm_Cue
root=WavePlayer_1
nodes [{id=WavePlayer_1, properties {soundWaveAssetPtr=&/ProximityMine_Disarm_01}}]
properties {
 attenuationSettings=/Game/Sounds/Attenuation/Fort_Weapon_Proximity_Mine_Att
 maxDistance=2800
 soundClassObject=&Fort_Loot_Collect
 volumeMultiplier=0.5
}
