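# Sound cue for dropping the proximity mine gadget, played as a quiet, spatialized one-shot.
# A random node picks one of three equally weighted drop wave variations and feeds it through a modulator at the root.
# It uses the shared weapon drop attenuation and the loot collect sound class.
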SoundCue /Game/Sounds/Fort_Gadgets/ProximityMine2/ProximityMine_Drop_Cue
root=Modulator_1
nodes [
 {id=Modulator_1, children [Random_0]}
 {id=Random_0, children [WavePlayer_1,WavePlayer_2,WavePlayer_5], properties {weights [1,1,1]}}
 {id=WavePlayer_1, properties {soundWaveAssetPtr=&/ProximityMine_Drop_01}}
 {id=WavePlayer_2, properties {soundWaveAssetPtr=&/ProximityMine_Drop_02}}
 {id=WavePlayer_5, properties {soundWaveAssetPtr=&/ProximityMine_Drop_03}}
]
properties {
 attenuationSettings=&Fort_Weapon_Drop_Att
 maxDistance=3000
 soundClassObject=&Fort_Loot_Collect
 volumeMultiplier=0.3
}
