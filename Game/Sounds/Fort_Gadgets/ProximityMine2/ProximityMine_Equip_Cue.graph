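# Sound cue for equipping the proximity mine gadget, played as player foley with close-prop attenuation.
# A single equip wave feeds a modulator that adds slight random volume gain, then a fixed short delay before playback.
# Concurrency is overridden to block new instances rather than stop existing ones.
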SoundCue /Game/Sounds/Fort_Gadgets/ProximityMine2/ProximityMine_Equip_Cue
root=Delay_0
nodes [
 {id=Delay_0, children [Modulator_0], properties {delayMax=0.05, delayMin=0.05}}
 {id=Modulator_0, children [WavePlayer_1], properties {volumeMax=1.2, volumeMin=1.1}}
 {id=WavePlayer_1, properties {soundWaveAssetPtr=&/ProximityMine_Equip_B_04}}
]
properties {
 attenuationSettings=/Game/Sounds/Attenuation/Fort_Prop_Close_Att
 bOverrideConcurrency=true
 maxDistance=1800
 soundClassObject=&Fort_Player_Foley
 volumeMultiplier=1
 concurrencyOverrides {resolutionRule=PreventNew}
}
