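# Sound cue for the Proximity Mine explosion, played through the large explosives sound class.
# Each branch mixes a random close explosion with a quieter random distant one, each layer with its own modulation and attenuation.
# A quality switch gives the first level three variants per layer, while the other levels share a branch with two.
# A concurrency override allows one instance per owner, stopping the oldest.
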SoundCue /Game/Sounds/Fort_Gadgets/ProximityMine2/ProximityMine_Explosion_Cue
root=QualityLevel_0
nodes [
 {
  id=Attenuation_2
  children [Modulator_1]
  properties {attenuationSettings=&Fort_Weapon_Projectile_Medium_Mono_Athena_Att}
 }
 {id=Attenuation_3, children [Modulator_0], properties {attenuationSettings=&Fort_Weapon_Projectile_Close_Athena_Att}}
 {
  id=Attenuation_4
  children [Modulator_3]
  properties {attenuationSettings=&Fort_Weapon_Projectile_Medium_Mono_Athena_Att}
 }
 {id=Attenuation_5, children [Modulator_2], properties {attenuationSettings=&Fort_Weapon_Projectile_Close_Athena_Att}}
 {id=Mixer_2, children [Attenuation_3, Attenuation_2], properties {inputVolume [1,0.7]}}
 {id=Mixer_4, children [Attenuation_5, Attenuation_4], properties {inputVolume [1,0.7]}}
 {id=Modulator_0, children [Random_0]}
 {id=Modulator_1, children [Random_1], properties {volumeMax=0.9, volumeMin=0.8}}
 {id=Modulator_2, children [Random_2]}
 {id=Modulator_3, children [Random_3], properties {volumeMax=0.9, volumeMin=0.8}}
 {id=QualityLevel_0, children [Mixer_2,Mixer_4,Mixer_4,Mixer_4]}
 {id=Random_0, children [WavePlayer_0,WavePlayer_1,WavePlayer_2], properties {weights [1,1,1]}}
 {id=Random_1, children [WavePlayer_3,WavePlayer_4,WavePlayer_5], properties {weights [1,1,1]}}
 {id=Random_2, children [WavePlayer_7,WavePlayer_8], properties {weights [1,1]}}
 {id=Random_3, children [WavePlayer_9, WavePlayer_10], properties {weights [1,1]}}
 {id=WavePlayer_0, properties {soundWaveAssetPtr=&/ProximityMine_Explosion_01}}
 {id=WavePlayer_1, properties {soundWaveAssetPtr=&/ProximityMine_Explosion_02}}
 {id=WavePlayer_10, properties {soundWaveAssetPtr=&/ProximityMine_Explosion_Distant_03}}
 {id=WavePlayer_2, properties {soundWaveAssetPtr=&/ProximityMine_Explosion_03}}
 {id=WavePlayer_3, properties {soundWaveAssetPtr=&/ProximityMine_Explosion_Distant_01}}
 {id=WavePlayer_4, properties {soundWaveAssetPtr=&/ProximityMine_Explosion_Distant_02}}
 {id=WavePlayer_5, properties {soundWaveAssetPtr=&ProximityMine_Explosion_Distant_03}}
 {id=WavePlayer_7, properties {soundWaveAssetPtr=&ProximityMine_Explosion_01}}
 {id=WavePlayer_8, properties {soundWaveAssetPtr=&ProximityMine_Explosion_02}}
 {id=WavePlayer_9, properties {soundWaveAssetPtr=&ProximityMine_Explosion_Distant_02}}
]
properties {
 bOverrideConcurrency=true
 maxDistance=10579.2871
 soundClassObject=/Game/Sounds/FortSoundClassesAndModes/Fort_Explosives_Large_PSM
 volumeMultiplier=0.5661
 concurrencyOverrides {bLimitToOwner=true, maxCount=1, resolutionRule=StopOldest}
}
