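# Sound cue for the pre-throw sound of the Proximity Mine gadget, classed as player foley.
# A random node picks among three equally weighted wave variations, then a modulator varies the pitch slightly on each play.
# It plays quietly with close-range prop attenuation and overrides concurrency to block new instances once the limit is reached.
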SoundCue /Game/Sounds/Fort_Gadgets/ProximityMine2/ProximityMine_PreThrow_Cue
root=Modulator_0
nodes [
 {id=Modulator_0, children [Random_0], properties {pitchMax=1.2, pitchMin=0.9}}
 {id=Random_0, children [WavePlayer_0,WavePlayer_2,WavePlayer_3], properties {weights [1,1,1]}}
 {id=WavePlayer_0, properties {soundWaveAssetPtr=&/ProximityMine_PreThrow_01}}
 {id=WavePlayer_2, properties {soundWaveAssetPtr=&/ProximityMine_PreThrow_02}}
 {id=WavePlayer_3, properties {soundWaveAssetPtr=&/ProximityMine_PreThrow_03}}
]
properties {
 attenuationSettings=/Game/Sounds/Attenuation/Fort_Prop_Close_Att
 bOverrideConcurrency=true
 maxDistance=1800
 soundClassObject=&Fort_Player_Foley
 volumeMultiplier=0.3
 concurrencyOverrides {resolutionRule=PreventNew}
}
