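# Sound cue for the spin sound of the proximity mine gadget.
# A random node picks one of three spin wave variations with equal weight on each play.
# It uses the proximity mine attenuation asset and the gadget items sound class.
# Its concurrency override prevents new instances from starting once the limit is reached, rather than cutting off playing ones.
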SoundCue /Game/Sounds/Fort_Gadgets/ProximityMine2/ProximityMine_Spin_Cue
root=Random_0
nodes [
 {id=Random_0, children [WavePlayer_1,WavePlayer_2,WavePlayer_3], properties {weights [1,1,1]}}
 {id=WavePlayer_1, properties {soundWaveAssetPtr=&/ProximityMine_Spin_01}}
 {id=WavePlayer_2, properties {soundWaveAssetPtr=&/ProximityMine_Spin_02}}
 {id=WavePlayer_3, properties {soundWaveAssetPtr=&/ProximityMine_Spin_03}}
]
properties {
 attenuationSettings=/Game/Sounds/Attenuation/Fort_Weapon_Proximity_Mine_Att
 bOverrideConcurrency=true
 maxDistance=2800
 soundClassObject=&Fort_Gadgets_Items
 concurrencyOverrides {resolutionRule=PreventNew}
}
